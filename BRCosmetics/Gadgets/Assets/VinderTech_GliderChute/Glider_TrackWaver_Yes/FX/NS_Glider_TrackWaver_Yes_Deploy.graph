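# Niagara one-shot deploy effect for the TrackWaver Yes glider cosmetic: a short, warm orange burst of energy, flare, glow waves and motes.
# NE_Energy and NE_Glow_Waves spawn via a shared socket reader; NE_Motes fly outward with drag.
# Energy and Flare use spawn groups as VisibilityTag, combining mesh and sprite renderers in each emitter with curve-driven scale and color.
# Energy alpha also dims by time of day for one group.
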NiagaraSystem /BRCosmetics/Gadgets/Assets/VinderTech_GliderChute/Glider_TrackWaver_Yes/FX/NS_Glider_TrackWaver_Yes_Deploy
dependencies {
 usedDataInterfaces [
  &NiagaraDataInterfaceCurve
  &NiagaraDataInterfaceColorCurve
  &NiagaraDataInterfaceVectorCurve
  &NiagaraDataInterfaceVector2DCurve
  &NiagaraDataInterfaceSocketReader
 ]
 usedDynamicInputs []
 usedModules [
  &self:SystemSpawnScript.SystemScriptSource.SystemScriptGraph.NiagaraNodeAssignment_2.SetVariables_#1
  &SystemState
 ]
}
emitters {
 NE_Energy {
  emitterUpdateScript [
   &EmitterState(Inactive Response=::NewEnumerator1, Life Cycle Mode=::NewEnumerator1, Loop Behavior=::NewEnumerator1, Loop Duration=0.5, UseLoopDelay=false)
   &SpawnBurst_Instantaneous(Use Spawn Probability=false, UseLoopCountLimit=false)
   &SpawnBurst_Instantaneous(Spawn Group=1, Use Spawn Probability=false, UseLoopCountLimit=false) {module=SpawnBurst_Instantaneous001}
  ]
  particleSpawnScript [
   /SetVariables(Particles.VisibilityTag=&SpawnGroup()) {isSetParameters=true}
   &InitializeParticle(
    Color=<rgba 2 0.3333 0 1>
    Lifetime=0.4
    Mesh Scale Mode=::NewEnumerator3
    Mesh Uniform Scale=0.2
    Sprite Size Mode=::NewEnumerator3
    Uniform Sprite Size=800
    UsePositionOffset=false
    Write Material Random=true
   )
   &SocketLocation(
    SamplePosOffset=true
    Sampled Position Offset=<0 -15 0>
    Socket Reader=$System.Socket Reader
    Socket Sampling Mode=::NewEnumerator3
    VelocitySpeedLimit=false
   )
   &InitialMeshOrientation(Mesh Orientation Mode=::NewEnumerator2, Use Rotation Vector=false)
   &CameraOffset(Camera Offset Amount=5)
  ]
  particleUpdateScript [
   &ParticleState()
   &ScaleColor(
    Scale Alpha=&MergeHelper_FloatSwitch(
     0=&Multiply_Float(
      A=&FloatFromCurve(FloatCurve=DataInterface() {propertyValues {Curve {keys [
           {arriveTangent=-0.9, interpMode=RCIM_Cubic, time=0, value=0}
           {arriveTangent=0.8333, interpMode=RCIM_Cubic, leaveTangent=0.8333, tangentMode=RCTM_Break, time=0.1, value=0.5}
           {arriveTangent=-1.6178, interpMode=RCIM_Cubic, leaveTangent=-1.6178, tangentMode=RCTM_User, time=0.6, value=0.5}
           {arriveTangent=-1.25, interpMode=RCIM_Cubic, leaveTangent=-1.25, time=0.7983, value=0.415}
           <at 1 0 RCIM_Cubic>
          ]}}})
      B=&FNTimeOfDayScalar(Night Scale=0.5, Morning Scale=0.5, Evening Scale=0.5)
     )
     1=&FloatFromCurve(FloatCurve=DataInterface() {propertyValues {Curve {keys [
          {interpMode=RCIM_Cubic, tangentMode=RCTM_User, time=0, value=0}
          {interpMode=RCIM_Cubic, tangentMode=RCTM_User, time=0.5, value=1}
          {interpMode=RCIM_Cubic, tangentMode=RCTM_User, time=1, value=0}
         ]}}})
     2=?
     3=?
     Index=$Particles.VisibilityTag
     Num=::NewEnumerator0
    )
    Scale RGB=&MakeVectorFromLinearColorRGB(Color=&ColorFromCurve(ColorCurve=DataInterface() {propertyValues {
        AlphaCurve {keys []}
        BlueCurve {keys [<at 0.1995 0 RCIM_Cubic>, {arriveTangent=0.0145, interpMode=RCIM_Cubic, leaveTangent=0.0145, time=0.4322, value=0.75}, <at 0.6627 0.0067>]}
        GreenCurve {keys [
          <at 0.1995 1.6667 RCIM_Cubic>
          {arriveTangent=-1.7966, interpMode=RCIM_Cubic, leaveTangent=-1.7966, time=0.4322, value=0.9516}
          <at 0.6627 0.8345>
         ]}
        RedCurve {keys [<at 0.1995 2 RCIM_Cubic>, {arriveTangent=-2.1588, interpMode=RCIM_Cubic, leaveTangent=-2.1588, time=0.4322, value=1}, <at 0.6627 1>]}
       }}))
    ScaleA=true
    ScaleRGB=true
   )
   &ScaleMeshSize(Scale Factor=&ScaleFloatByCurve(
     InitialValue:Vector3f=<1 1 1>
     ScaleCurve:NiagaraDataInterfaceVectorCurve=DataInterface() {propertyValues {XCurve {keys [<at 0 0.25>,<at 1 1>]}, YCurve {keys [<at 0 1.25>,<at 1 2.25>]}, ZCurve {keys [<at 0 1.25>,<at 1 2.25>]}}}
     CurveIndex:NiagaraFloat=$Particles.NormalizedAge
    ))
   &ScaleSpriteSize(Uniform Curve Sprite Scale=DataInterface() {propertyValues {Curve {
       defaultValue=0
       keys [{interpMode=RCIM_Cubic, tangentMode=RCTM_User, time=0, value=0}, {interpMode=RCIM_Cubic, tangentMode=RCTM_User, time=1, value=1}]
      }}})
  ]
  rendererClasses [&NiagaraMeshRendererProperties, &NiagaraSpriteRendererProperties]
  renderers [
   {rendererClass=&NiagaraMeshRendererProperties, rendererIndex=0}
   {rendererClass=&NiagaraSpriteRendererProperties, rendererIndex=1}
  ]
 }
 NE_Flare {
  emitterUpdateScript [
   &EmitterState(Life Cycle Mode=::NewEnumerator1, Loop Behavior=::NewEnumerator1, Loop Duration=2, UseLoopDelay=false)
   &SpawnBurst_Instantaneous(Use Spawn Probability=false, UseLoopCountLimit=false) {module=SpawnBurst_Instantaneous002}
   &SpawnBurst_Instantaneous(Spawn Group=1, Use Spawn Probability=false, UseLoopCountLimit=false)
   &SpawnBurst_Instantaneous(Spawn Group=2, Use Spawn Probability=false, UseLoopCountLimit=false) {module=SpawnBurst_Instantaneous001}
   &SpawnBurst_Instantaneous(Spawn Group=3, Use Spawn Probability=false, UseLoopCountLimit=false) {module=SpawnBurst_Instantaneous003}
  ]
  particleSpawnScript [
   /SetVariables(Particles.VisibilityTag=&SpawnGroup()) {isSetParameters=true}
   &InitializeParticle(
    Color=&MergeHelper_ColorSwitch(
     0=<rgba 2 0.5227 0.2272 1.2>
     1=<rgba 1 0.2942 0.2009 0.7>
     2=<rgba 2.5 0.4167 0 1.25>
     3=?
     Index=&SpawnGroup()
     Num=::NewEnumerator1
    )
    Lifetime=&MergeHelper_FloatSwitch(0=0.5, 1=0.5, 2=0.7, 3=0.85, Index=&SpawnGroup(), Num=::NewEnumerator2)
    Mesh Scale Mode=::NewEnumerator3
    Mesh Uniform Scale=1.2
    Position Offset=&MergeHelper_VectorSwitch(0=<0 -10 5>, 1=<0 -10 5>, 2=<0 -10 3>, 3=<0 -10 10>, 4=<0 -10 0>, Index=&SpawnGroup(), Num=::NewEnumerator3)
    Sprite Rotation Angle=90
    Sprite Rotation Mode=::NewEnumerator0
    Sprite Size=&MergeHelper_Vector2Switch(0=<160 260>, 1=<160 260>, 2=<0 0>, 3=<250 250>, 4=?, Index=&SpawnGroup(), Num=::NewEnumerator2)
    Sprite Size Mode=::NewEnumerator4
    UsePositionOffset=true
    Write Material Random=true
   )
   &CameraOffset(Camera Offset Amount=10)
  ]
  particleUpdateScript [
   &ParticleState()
   &ScaleMeshSize(Scale Factor=&VectorFromCurve(
     VectorCurve=DataInterface() {propertyValues {
       XCurve {
        defaultValue=0
        keys [
         {arriveTangent=7.2229, interpMode=RCIM_Cubic, leaveTangent=7.2229, tangentMode=RCTM_User, time=0, value=0}
         {interpMode=RCIM_Cubic, tangentMode=RCTM_User, time=0.1515, value=0.7519}
         {interpMode=RCIM_Cubic, tangentMode=RCTM_User, time=0.9849, value=1.05}
        ]
       }
       YCurve {keys [
         {arriveTangent=8.0619, interpMode=RCIM_Cubic, leaveTangent=8.0619, leaveTangentWeight=0.3333, tangentMode=RCTM_User, time=0, value=0}
         {arriveTangentWeight=0.069, interpMode=RCIM_Cubic, leaveTangentWeight=0.2643, tangentMode=RCTM_User, time=0.1584, value=0.6945}
         {arriveTangentWeight=0.2556, interpMode=RCIM_Cubic, leaveTangentWeight=0.0087, tangentMode=RCTM_User, time=0.988, value=0}
        ]}
       ZCurve {
        defaultValue=0
        keys [
         {arriveTangent=9.0638, interpMode=RCIM_Cubic, leaveTangent=9.0638, tangentMode=RCTM_User, time=0, value=0}
         {arriveTangentWeight=0.069, interpMode=RCIM_Cubic, tangentMode=RCTM_User, time=0.141, value=0.7268}
         {arriveTangentWeight=0.2593, interpMode=RCIM_Cubic, tangentMode=RCTM_User, time=0.9849, value=0}
        ]
       }
      }}
     Scale Curve=<2 2 2>
    ))
   &ScaleSpriteSize(
    Non-Uniform Curve Scale=<1.5 1>
    Non-Uniform Curve Sprite Scale=DataInterface() {propertyValues {
      XCurve {keys [
        {arriveTangent=41.408, interpMode=RCIM_Cubic, leaveTangent=41.408, tangentMode=RCTM_User, time=0, value=0.1}
        {interpMode=RCIM_Cubic, tangentMode=RCTM_User, time=0.0457, value=0.9462}
        {arriveTangent=-11.5805, interpMode=RCIM_Cubic, leaveTangent=-0.9291, tangentMode=RCTM_Break, time=0.1042, value=0.4729}
        {arriveTangent=-0.1541, interpMode=RCIM_Cubic, leaveTangent=-0.1541, tangentMode=RCTM_User, time=1, value=0}
       ]}
      YCurve {keys [
        {interpMode=RCIM_Cubic, leaveTangent=4.0442, tangentMode=RCTM_Break, time=0, value=0.1}
        {arriveTangent=4.8062, interpMode=RCIM_Cubic, leaveTangent=4.8062, time=0.1027, value=0.4785}
        {interpMode=RCIM_Cubic, tangentMode=RCTM_User, time=0.1788, value=0.9593}
        {arriveTangent=-0.7572, interpMode=RCIM_Cubic, leaveTangent=-0.7572, tangentMode=RCTM_User, time=0.3212, value=0.2892}
        <at 0.9979 0.0156>
       ]}
     }}
    Scale Sprite Size Mode=::NewEnumerator3
   )
   &ScaleColor(
    Scale Alpha=&FloatFromCurve(FloatCurve=DataInterface() {propertyValues {Curve {keys [
         <at -0.0046 1.0113 RCIM_Cubic>
         {arriveTangent=-1.0036, interpMode=RCIM_Cubic, leaveTangent=-1.0036, time=0.5003, value=0.2765}
         <at 1.0011 0.002 RCIM_Cubic>
        ]}}})
    ScaleA=true
    ScaleRGB=false
   )
   &SolveForcesAndVelocity(Clamp Velocity=false, Limit Acceleration=false)
   /SetVariables(Particles.Mesh Scale Color:LinearColor=&ColorFromCurve(ColorCurve=DataInterface() {propertyValues {
       AlphaCurve {keys [<at 0.2748 0.6 RCIM_Cubic>,<at 0.5629 0.27>,<at 1 0 RCIM_Cubic>]}
       BlueCurve {keys [
         <at 0 0.0526 RCIM_Cubic>
         <at 0.1114 0.0703>
         <at 0.3448 0.1023>
         <at 0.6276 0>
         {arriveTangent=0.2881, interpMode=RCIM_Cubic, leaveTangent=0.2881, time=0.7762, value=0.0577}
         <at 0.945 0.0914>
        ]}
       GreenCurve {keys [
         <at 0 0.2105 RCIM_Cubic>
         <at 0.1114 0.2954>
         <at 0.3448 0.2288>
         <at 0.6276 0.0856>
         {arriveTangent=0.4954, interpMode=RCIM_Cubic, leaveTangent=0.4954, time=0.7762, value=0.1676}
         <at 0.945 0.2429>
        ]}
       RedCurve {keys [
         <at 0 1 RCIM_Cubic>
         <at 0.1114 2>
         <at 0.3448 2>
         <at 0.6276 1.2845>
         {arriveTangent=-0.8963, interpMode=RCIM_Cubic, leaveTangent=-0.8963, time=0.7762, value=1}
         <at 0.945 1>
        ]}
      }})) {isSetParameters=true}
   /SetVariables(
    Particles.Corona Scale Color:LinearColor=&MakeLinearColorFromVectorAndFloat(
     "Vector (RGB)"=&MakeVectorFromLinearColorRGB(Color=<rgba 1.5 0.2978 0.0573 2>)
     "Float (Alpha)"=&FloatFromCurve(FloatCurve=DataInterface() {propertyValues {Curve {
         defaultValue=1
         keys [
          {arriveTangent=20.985, interpMode=RCIM_Cubic, leaveTangent=20.985, tangentMode=RCTM_User, time=0, value=0}
          {arriveTangent=-3.9091, interpMode=RCIM_Cubic, leaveTangent=-3.9091, tangentMode=RCTM_User, time=0.0828, value=1.004}
          {arriveTangent=-1.5525, interpMode=RCIM_Cubic, leaveTangent=-1.5525, tangentMode=RCTM_User, time=0.2526, value=0.4508}
          {arriveTangent=-0.0405, interpMode=RCIM_Cubic, leaveTangent=-0.0405, tangentMode=RCTM_User, time=1, value=0}
         ]
        }}})
    )
    Particles.Corona Scale Sprite Size:Vector2f=&Multiply_Vector2D(
     A=$Particles.Initial.SpriteSize
     B=&Vector2DFromFloat(Value=&FloatFromCurve(
       FloatCurve=DataInterface() {propertyValues {Curve {
          defaultValue=1
          keys [
           {arriveTangent=5.8075, interpMode=RCIM_Cubic, leaveTangent=5.8075, tangentMode=RCTM_User, time=0, value=0}
           {arriveTangent=1.1179, interpMode=RCIM_Cubic, leaveTangent=1.1179, tangentMode=RCTM_User, time=0.2465, value=0.8111}
           {arriveTangent=0.0224, interpMode=RCIM_Cubic, leaveTangent=0.0224, tangentMode=RCTM_User, time=1, value=1}
          ]
         }}}
       Scale Curve=0.73
      ))
    )
   ) {isSetParameters=true}
  ]
  rendererClasses [&NiagaraSpriteRendererProperties, &NiagaraMeshRendererProperties]
  renderers [
   {rendererClass=&NiagaraSpriteRendererProperties}
   {rendererClass=&NiagaraSpriteRendererProperties, rendererIndex=1}
   {rendererClass=&NiagaraMeshRendererProperties, rendererIndex=2}
   {rendererClass=&NiagaraSpriteRendererProperties, rendererIndex=3}
  ]
 }
 NE_Glow_Waves {
  emitterUpdateScript [
   &EmitterState(Inactive Response=::NewEnumerator1, Life Cycle Mode=::NewEnumerator1, Loop Behavior=::NewEnumerator1, Loop Duration=1, UseLoopDelay=false)
   &SpawnBurst_Instantaneous(Spawn Count=2, Use Spawn Probability=false, UseLoopCountLimit=false)
  ]
  particleSpawnScript [
   &InitializeParticle(
    Color=<rgba 0.4 0.1333 0.08 0.4>
    Lifetime=0.5
    Position Offset=<0 -15 0>
    Sprite Size=<111 275>
    Sprite Size Mode=::NewEnumerator4
    UsePositionOffset=true
    Write Material Random=true
   )
   &SocketLocation(SamplePosOffset=false, Socket Reader=$System.Socket Reader, Socket Sampling Mode=::NewEnumerator3, VelocitySpeedLimit=false)
  ]
  particleUpdateScript [
   &ParticleState()
   &ScaleColor(
    Scale Mode=::NewEnumerator1
    Scale RGBA=&MakeVector4_FromFloat(Float=&FloatFromCurve(FloatCurve=DataInterface() {propertyValues {Curve {keys [<at 0 0>, {arriveTangent=-3.0726, interpMode=RCIM_Cubic, leaveTangent=-3.0726, tangentMode=RCTM_User, time=0.05, value=1}, <at 1 0>]}}}))
    ScaleRGBA=true
   )
   /SetVariables(Particles.SpriteRotation:NiagaraFloat=&Multiply_FloatByInt(Float=&Add_Float(A=/Game/Effects/Niagara/DynamicInputs/Math/FN_NormalizedParticleCount(), B=0.25), Integer=360)) {isSetParameters=true}
   &ScaleSpriteSize(
    Scale Factor=&Vector2DFromCurve(
     Vector2Curve=DataInterface() {propertyValues {
       XCurve {keys [
         {arriveTangent=6.2067, interpMode=RCIM_Cubic, leaveTangent=6.2067, tangentMode=RCTM_User, time=0, value=0}
         {tangentMode=RCTM_User, time=1, value=6}
        ]}
       YCurve {keys [
         {arriveTangent=2.8894, interpMode=RCIM_Cubic, leaveTangent=2.8894, tangentMode=RCTM_User, time=0, value=0}
         {tangentMode=RCTM_User, time=1, value=1.5}
        ]}
      }}
     ScaleCurve=<1 1>
    )
    Scale Sprite Size Mode=::NewEnumerator2
   )
   &SpriteFacingAndAlignment(Sprite Alignment=<0 0 1>, Sprite Facing=<0 -1 0.1>, WriteAlignment=true, WriteFacing=true)
  ]
  rendererClasses [&NiagaraSpriteRendererProperties]
  renderers [{rendererClass=&NiagaraSpriteRendererProperties}]
 }
 NE_Motes {
  emitterUpdateScript [
   &EmitterState(Inactive Response=::NewEnumerator1, Life Cycle Mode=::NewEnumerator1, Loop Behavior=::NewEnumerator1, Loop Duration=1, UseLoopDelay=false)
   &SpawnBurst_Instantaneous(Spawn Count=64, Use Spawn Probability=false, UseLoopCountLimit=false)
  ]
  particleSpawnScript [
   &InitializeParticle(
    Color=<rgba 1.5 0.6875 0.525 1>
    Lifetime Max=0.7
    Lifetime Min=0.4
    Lifetime Mode=::NewEnumerator1
    Sprite Size Mode=::NewEnumerator1
    Uniform Sprite Size Max=4
    Uniform Sprite Size Min=2
    UsePositionOffset=false
    Write Material Random=true
   )
   &SphereLocation(
    MaskSpawnGroup=false
    Override Seed=false
    Sphere Radius=1
    Surface Only=false
    UseNonUniformScale=false
    UseOffset=false
    UseSphereOrientation=false
   )
   &SubUVAnimation(End Frame=8, SubUV Animation Mode:ENiagara_SubUVLookupMode=::NewEnumerator2)
   &AddVelocityFromPoint(Module.Use Falloff Distance=false, Velocity Strength=&RandomRangeFloat(Minimum=150, Maximum=450))
  ]
  particleUpdateScript [
   &Drag(Drag=4, Use Linear Drag=true, Use Rotational Drag=false)
   &SolveForcesAndVelocity(Clamp Velocity=false, Limit Acceleration=false)
   &ParticleState()
   &ScaleColor(
    Scale Mode=::NewEnumerator1
    Scale RGBA=&MakeVector4_FromFloat(Float=&FloatFromCurve(FloatCurve=DataInterface() {propertyValues {Curve {keys [
          {arriveTangent=13.1284, interpMode=RCIM_Cubic, leaveTangent=13.1284, tangentMode=RCTM_User, time=0, value=0}
          {arriveTangent=-3.0086, interpMode=RCIM_Cubic, leaveTangent=-3.0086, tangentMode=RCTM_User, time=0.1, value=1}
          <at 1 0>
         ]}}}))
    ScaleRGBA=true
   )
  ]
  rendererClasses [&NiagaraSpriteRendererProperties]
  renderers [{rendererClass=&NiagaraSpriteRendererProperties}]
 }
}
summary {
 systemName=NS_Glider_TrackWaver_Yes_Deploy
 emitters [
  {emitterName=NE_Motes, rendererClasses [&NiagaraSpriteRendererProperties]}
  {emitterName=NE_Energy, rendererClasses [&NiagaraMeshRendererProperties, &NiagaraSpriteRendererProperties]}
  {emitterName=NE_Glow_Waves, rendererClasses [&NiagaraSpriteRendererProperties]}
  {emitterName=NE_Flare, rendererClasses [&NiagaraSpriteRendererProperties, &NiagaraMeshRendererProperties]}
 ]
 userVariables [
  User.bIsFrontEnd {type=NiagaraBool}
  User.bIsFrontEndPreview {type=NiagaraBool}
  User.bIsFullyDeployed {type=NiagaraBool}
  User.ForwardDot {type=NiagaraFloat, defaultValue=0, description=""}
  User.IsFrontEnd {type=NiagaraBool}
  User.IsWeakPoint {type=NiagaraBool}
  User.RibbonIDOffset {type=NiagaraInt32}
  User.RightDot {type=NiagaraFloat, defaultValue=0, description=""}
  User.Scale {type=NiagaraFloat, defaultValue=1, description=""}
  User.VariantID {type=NiagaraFloat, defaultValue=0, description=""}
 ]
}
