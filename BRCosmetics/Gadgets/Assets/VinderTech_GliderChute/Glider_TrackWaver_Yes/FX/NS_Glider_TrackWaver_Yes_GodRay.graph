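# Niagara system for the TrackWaver Yes glider cosmetic, rendering warm god rays with swirling dust motes.
# NE_GodRay spawns mesh shafts with alpha scaled by time of day; NE_Ray burst-spawns a fading mesh with a material parameter switched by front-end preview.
# NE_Dust_01 sprites move through gravity, vortex and curl noise, fading out near five seconds of emitter age.
# A random VisibilityTag splits dust into two variants of size, flicker and hot-spiked colour curves.
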NiagaraSystem /BRCosmetics/Gadgets/Assets/VinderTech_GliderChute/Glider_TrackWaver_Yes/FX/NS_Glider_TrackWaver_Yes_GodRay
dependencies {
 usedDataInterfaces [&NiagaraDataInterfaceCurve, &NiagaraDataInterfaceColorCurve]
 usedDynamicInputs []
 usedModules [&SystemState]
}
emitters {
 NE_Dust_01 {
  emitterUpdateScript [
   &EmitterState(Life Cycle Mode=::NewEnumerator1, Loop Behavior=::NewEnumerator1, Loop Duration=2, UseLoopDelay=false)
   &SpawnRate(
    Spawn Probability=&FloatFromCurve(
     FloatCurve=DataInterface() {propertyValues {Curve {
        defaultValue=1
        keys [{interpMode=RCIM_Cubic, tangentMode=RCTM_User, time=0, value=1}, {interpMode=RCIM_Cubic, tangentMode=RCTM_User, time=1, value=0}]
       }}}
     CurveIndex=$Emitter.NormalizedLoopAge
    )
    SpawnRate=50
    Use Spawn Probability=true
   )
  ]
  particleSpawnScript [
   /SetVariables(Particles.VisibilityTag=&RandomRangeInt(Minimum=0, Maximum=1)) {isSetParameters=true}
   &InitializeParticle(
    Color Mode=::NewEnumerator0
    Lifetime Max=3
    Lifetime Min=1.5
    Lifetime Mode=::NewEnumerator1
    Mass Mode=::NewEnumerator3
    Position Offset=<0 0 50>
    Sprite Rotation Mode=::NewEnumerator2
    Sprite Size Mode=::NewEnumerator3
    Uniform Sprite Size=&MaskFloat(
     Mask Source=1
     Mask Parameter=$Particles.VisibilityTag
     Mask=0
     Masked Float=&RandomRangeFloat(Minimum=3, Maximum=5)
     Passthrough Float=&RandomRangeFloat(Minimum=2, Maximum=3)
    )
    UsePositionOffset=true
    Write Material Random=true
   )
   &ShapeLocation(
    MaskSpawnGroup=false
    Non Uniform Scale=<0.75 0.75 1.6>
    Offset=<0 7 190>
    Offset Mode=::NewEnumerator0
    Override Seed=false
    Sphere Radius=50
    UseOwnerScale=true
   )
   &CameraOffset(Camera Offset Amount=&RandomRangeFloat(Minimum=-10, Maximum=-20))
  ]
  particleUpdateScript [
   &ParticleState()
   &ScaleSpriteSize(
    Uniform Curve Scale=&RandomRangeFloat(Minimum=1, Maximum=1.6)
    Uniform Curve Sprite Scale=DataInterface() {propertyValues {Curve {
       defaultValue=0
       keys [
        {interpMode=RCIM_Cubic, tangentMode=RCTM_User, time=0, value=0}
        {interpMode=RCIM_Cubic, tangentMode=RCTM_User, time=0.5, value=1}
        {tangentMode=RCTM_User, time=1, value=0}
       ]
      }}}
   )
   &GravityForce(Gravity=&RandomRangeVector(Minimum=<0 0 -60>, Maximum=<0 0 -120>))
   &VortexForce(
    Origin Pull Amount=10
    UseOriginOffset=false
    Vortex Force Amount=&RandomRangeFloat(Minimum=200, Maximum=400)
   )
   &CurlNoiseForce(
    Bias Noise=false
    Noise Frequency=&RandomRangeFloat(Minimum=30, Maximum=70)
    Noise Strength=&RandomRangeFloat(Minimum=20, Maximum=50)
    Pan Noise=false
   )
   &Drag(Drag=&RandomRangeFloat(Minimum=1.5, Maximum=2.5), Use Linear Drag=true, Use Rotational Drag=false)
   &SolveForcesAndVelocity(Clamp Velocity=false, Limit Acceleration=false)
   &ScaleSpriteSizeBySpeed(Max Scale Factor=<1 1.3>, Min Scale Factor=<0.5 1>)
   &ScaleColor(
    Scale Alpha=&Multiply_Float(
     A=&FloatFromCurve(FloatCurve=DataInterface() {propertyValues {Curve {keys [<at 4.8 1>,<at 5.1 0>]}}}, CurveIndex=$Emitter.Age)
     B=&MaskFloat(
      Mask Source=1
      Mask Parameter=$Particles.VisibilityTag
      Mask=0
      Masked Float=&FloatFromCurve(
       FloatCurve=DataInterface() {propertyValues {Curve {
          defaultValue=1
          keys [
           {interpMode=RCIM_Cubic, tangentMode=RCTM_User, time=0, value=0}
           {interpMode=RCIM_Cubic, tangentMode=RCTM_User, time=0.5, value=1}
           {tangentMode=RCTM_User, time=1, value=0}
          ]
         }}}
       Scale Curve=1.6
      )
      Passthrough Float=&FloatFromCurve(FloatCurve=DataInterface() {propertyValues {Curve {
          defaultValue=1
          keys [
           {interpMode=RCIM_Cubic, tangentMode=RCTM_User, time=0, value=0}
           {arriveTangent=1.0582, interpMode=RCIM_Cubic, leaveTangent=1.0582, time=0.1807, value=0.7046}
           {interpMode=RCIM_Cubic, tangentMode=RCTM_User, time=0.4096, value=0.4335}
           {arriveTangent=-0.7343, interpMode=RCIM_Cubic, leaveTangent=-0.7343, time=0.6593, value=1}
           {tangentMode=RCTM_User, time=1, value=0}
          ]
         }}})
     )
    )
    Scale RGB=&MaskVector(
     Mask Source=1
     Mask Parameter=$Particles.VisibilityTag
     Mask=0
     Masked Vector=&MakeVectorFromLinearColorRGB(Color=&MakeCustomLinearColorFromBool(
       Bool=&SetBoolByFloatComparison(A=0.55, B=&RandomRangeFloat(Minimum=0, Maximum=1))
       True Color=&ColorFromCurve(ColorCurve=DataInterface() {propertyValues {
          AlphaCurve {keys []}
          BlueCurve {keys [
            <at 0 0 RCIM_Cubic>
            {arriveTangent=0.3312, interpMode=RCIM_Cubic, leaveTangent=0.3312, time=0.2164, value=0.1685}
            <at 0.3292 0.109>
            <at 0.3597 0.128>
            <at 0.4157 0>
            <at 0.4751 0.1385>
            <at 0.535 0.1585>
            <at 0.6895 0.3359>
            <at 0.9946 0.0895>
           ]}
          GreenCurve {keys [
            <at 0 0.1315 RCIM_Cubic>
            {arriveTangent=0.3827, interpMode=RCIM_Cubic, leaveTangent=0.3827, time=0.2164, value=0.3348}
            <at 0.3292 0.2575>
            <at 0.3597 0.2733>
            <at 0.4157 2.0833>
            <at 0.4751 0.2821>
            <at 0.535 0.3268>
            <at 0.6895 0.4687>
            <at 0.9946 0.1957>
           ]}
          RedCurve {keys [<at 0 1 RCIM_Cubic>,<at 0.2164 1 RCIM_Cubic>,<at 0.3292 1>,<at 0.3597 1>,<at 0.4157 25>,<at 0.4751 1>,<at 0.535 1>,<at 0.6895 1>,<at 0.9946 1>]}
         }})
       False Color=&ColorFromCurve(ColorCurve=DataInterface() {propertyValues {
          AlphaCurve {keys []}
          BlueCurve {keys [<at 0.0037 0>,<at 0.1822 0.1585>,<at 0.2933 0.1493>,<at 0.3717 0>,<at 0.4446 0.1451>,<at 0.5143 0.1385>,<at 0.7211 0.2464>,<at 0.8387 0.2076>,<at 1 0>]}
          GreenCurve {keys [<at 0.0037 0.2>,<at 0.1822 0.2286>,<at 0.2933 0.3194>,<at 0.3717 5>,<at 0.4446 0.3161>,<at 0.5143 0.2821>,<at 0.7211 0.372>,<at 0.8387 0.3712>,<at 1 0.1806>]}
          RedCurve {keys [<at 0.0037 1>,<at 0.1822 1>,<at 0.2933 1>,<at 0.3717 25>,<at 0.4446 1>,<at 0.5143 1>,<at 0.7211 1>,<at 0.8387 1>,<at 1 1>]}
         }})
      ))
     Passthrough Vector=&MakeVectorFromLinearColorRGB(Color=&MakeCustomLinearColorFromBool(
       Bool=&SetBoolByFloatComparison(A=0.55, B=&RandomRangeFloat(Minimum=0, Maximum=1))
       True Color=&ColorFromCurve(ColorCurve=DataInterface() {propertyValues {
          AlphaCurve {keys []}
          BlueCurve {keys [<at 0.0091 0.0895>,<at 0.0734 0.3359>,<at 0.1365 0.0985>,<at 0.1779 0>,<at 0.2236 0>,<at 0.2454 0>,<at 0.3292 0.109>,<at 0.3597 0.128>,<at 0.4157 0>,<at 0.4751 0.1385>,<at 0.535 0.1585>,<at 0.7309 0.067>,<at 0.7701 0>,<at 0.806 0.2527>,<at 0.8812 0.2271>,<at 0.9585 0 RCIM_Cubic>]}
          GreenCurve {keys [<at 0.0091 0.1957>,<at 0.0734 0.4466>,<at 0.1365 0.2788>,<at 0.1779 6.25>,<at 0.2236 0.2>,<at 0.2454 0.2>,<at 0.3292 0.2872>,<at 0.3597 0.3024>,<at 0.4157 6.25>,<at 0.4751 0.3108>,<at 0.535 0.3268>,<at 0.7309 0.2536>,<at 0.7701 6.25>,<at 0.806 0.4022>,<at 0.8812 0.3817>,<at 0.9585 0.1167 RCIM_Cubic>]}
          RedCurve {keys [<at 0.0091 1>,<at 0.0734 1>,<at 0.1365 1>,<at 0.1779 25>,<at 0.2236 1>,<at 0.2454 1>,<at 0.3292 1>,<at 0.3597 1>,<at 0.4157 25>,<at 0.4751 1>,<at 0.535 1>,<at 0.7309 1>,<at 0.7701 25>,<at 0.806 1>,<at 0.8812 1>,<at 0.9585 1 RCIM_Cubic>]}
         }})
       False Color=&ColorFromCurve(ColorCurve=DataInterface() {propertyValues {
          AlphaCurve {keys []}
          BlueCurve {keys [<at 0.0037 0.2391>,<at 0.1822 0.1585>,<at 0.2933 0.1493>,<at 0.3717 0>,<at 0.4446 0.1451>,<at 0.5143 0.1385>,<at 0.7211 0.3716>,<at 0.8387 0.2075>,<at 1 0>]}
          GreenCurve {keys [<at 0.0037 0.3279>,<at 0.1822 0.3549>,<at 0.2933 0.3478>,<at 0.3717 6.25>,<at 0.4446 0.3445>,<at 0.5143 0.3395>,<at 0.7211 0.5182>,<at 0.8387 0.3>,<at 1 0.1167>]}
          RedCurve {keys [<at 0.0037 1>,<at 0.1822 1>,<at 0.2933 1>,<at 0.3717 25>,<at 0.4446 1>,<at 0.5143 1>,<at 0.7211 1>,<at 0.8387 1>,<at 1 1>]}
         }})
      ))
    )
    ScaleA=true
    ScaleRGB=true
   )
  ]
  rendererClasses [&NiagaraSpriteRendererProperties]
  renderers [
   {rendererClass=&NiagaraSpriteRendererProperties}
   {rendererClass=&NiagaraSpriteRendererProperties, rendererIndex=1}
  ]
 }
 NE_GodRay {
  emitterUpdateScript [
   &EmitterState(Life Cycle Mode=::NewEnumerator1, Loop Behavior=::NewEnumerator1, Loop Delay=0.1, Loop Duration=1, UseLoopDelay=true)
   &SpawnRate(SpawnRate=3, Use Spawn Probability=false)
  ]
  particleSpawnScript [
   &InitializeParticle(
    Color=<rgba 1 0.1167 0 1>
    Lifetime=2
    Mesh Scale Max=<2 2 1.5>
    Mesh Scale Min=<1.5 1.5 1.5>
    Mesh Scale Mode=::NewEnumerator2
    Position Offset=<0 0 350>
    Sprite Size Max=<100 -180>
    Sprite Size Min=<120 -150>
    Sprite Size Mode=::NewEnumerator2
    UsePositionOffset=true
    Write Material Random=true
   )
   &CameraOffset(Camera Offset Amount=-100)
  ]
  particleUpdateScript [
   &ParticleState()
   &ScaleMeshSize(Scale Factor=&MakeVector(X=1, Y=1, Z=1))
   &SolveForcesAndVelocity(Clamp Velocity=false, Limit Acceleration=false)
   &ScaleColor(
    Scale Alpha=&FloatFromCurve(
     FloatCurve=DataInterface() {propertyValues {Curve {keys [<at 0 0 RCIM_Cubic>,<at 0.4 1 RCIM_Cubic>,<at 1 0 RCIM_Cubic>]}}}
     Scale Curve=&FNTimeOfDayScalar(Night Scale=0.1, Morning Scale=0.25, Day Scale=0.5, Evening Scale=0.25)
    )
    ScaleA=true
    ScaleRGB=true
   )
   &DynamicMaterialParameters(
    Index 0 Param 1=&RandomRangeFloat(Minimum=0.05, Maximum=-0.05)
    Module.Index0Float0Write=true
    Module.Index0Float1Write=true
    Module.Index0Float2Write=true
    Module.Index0Float3Write=true
   )
  ]
  rendererClasses [&NiagaraMeshRendererProperties]
  renderers [{rendererClass=&NiagaraMeshRendererProperties, rendererIndex=0}]
 }
 NE_Ray {
  emitterUpdateScript [
   &EmitterState(
    Inactive Response=::NewEnumerator1
    Life Cycle Mode=::NewEnumerator1
    Loop Behavior=::NewEnumerator1
    Loop Delay=0.1
    Loop Duration=2
    UseLoopDelay=true
   )
   &SpawnBurst_Instantaneous(Use Spawn Probability=false, UseLoopCountLimit=false)
  ]
  particleSpawnScript [
   &InitializeParticle(
    Color=<rgba 2.5 0.6 0.125 1>
    Lifetime=2
    Mesh Scale=<1.5 1.5 1.5>
    Mesh Scale Mode=::NewEnumerator4
    Position Offset=<0 0 450>
    UsePositionOffset=true
    Write Material Random=true
   )
   &CameraOffset()
  ]
  particleUpdateScript [
   &ParticleState()
   &SolveForcesAndVelocity(Clamp Velocity=false, Limit Acceleration=false)
   &ScaleColor(
    Scale Alpha=&FloatFromCurve(FloatCurve=DataInterface() {propertyValues {Curve {keys [
         {interpMode=RCIM_Cubic, tangentMode=RCTM_User, time=0, value=0}
         {interpMode=RCIM_Cubic, tangentMode=RCTM_User, time=0.5, value=1}
         {interpMode=RCIM_Cubic, tangentMode=RCTM_User, time=1, value=0}
        ]}}})
    ScaleA=true
    ScaleRGB=true
   )
   &DynamicMaterialParameters(
    Index 0 Param 1=&MakeCustomFloatFromBool(Bool=$User.bIsFrontEndPreview, True Float=0.2, False Float=0.5)
    Index 0 Param 2=1
    Index 0 Param 3=1
    Index 0 Param 4=1
    Module.Param0WriteEnabled=true
    Module.Param1WriteEnabled=true
    Module.Param2WriteEnabled=true
    Module.Param3WriteEnabled=true
   )
  ]
  rendererClasses [&NiagaraMeshRendererProperties]
  renderers [{rendererClass=&NiagaraMeshRendererProperties, rendererIndex=0}]
 }
}
summary {
 systemName=NS_Glider_TrackWaver_Yes_GodRay
 emitters [
  {emitterName=NE_GodRay, rendererClasses [&NiagaraMeshRendererProperties]}
  {emitterName=NE_Dust_01, rendererClasses [&NiagaraSpriteRendererProperties]}
  {emitterName=NE_Ray, rendererClasses [&NiagaraMeshRendererProperties]}
 ]
 userVariables [
  User.bIsFrontEnd {type=NiagaraBool}
  User.bIsFrontEndPreview {type=NiagaraBool}
  User.bIsFullyDeployed {type=NiagaraBool}
  User.ForwardDot {type=NiagaraFloat, defaultValue=0, description=""}
  User.IsFrontEnd {type=NiagaraBool}
  User.IsWeakPoint {type=NiagaraBool}
  User.RibbonIDOffset {type=NiagaraInt32}
  User.RightDot {type=NiagaraFloat, defaultValue=0, description=""}
  User.Scale {type=NiagaraFloat, defaultValue=1, description=""}
  User.VariantID {type=NiagaraFloat, defaultValue=0, description=""}
 ]
}
